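# Physics asset for the Pizza Pit female medium player body cosmetic, covering only the dyn_cup bone and the nearby hip area.
# Three single-capsule bodies represent dyn_cup, pelvis and left thigh, with no other body parts included.
# The cup is constrained to the pelvis with a swing cone and zero twist range.
# The thigh joins the pelvis with free swing and twist, likely serving as collision for the dangling cup.
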PhysicsAsset /BRCosmetics/Characters/Player/Female/Medium/Bodies/F_MED_Pizza_Pit/Meshes/F_MED_Pizza_Pit_Physics
bodies {
 dyn_cup [{shapeType=Capsule, length=12.9596, radius=5.0171, center=<7.9759 -2.9164 -1.1639>, rotation=<rot 90 0 0>}]
 pelvis [{shapeType=Capsule, length=6.5866, radius=9.6432, center=<-0.2565 -1.9108 -0.3711>}]
 thigh_l [{shapeType=Capsule, length=14.6507, radius=8.0853, center=<-16.8006 1.2134 -0.4167>, rotation=<rot 0.2707 -87.8882 -87.4478>}]
}
constraints [
 {bone1Name=dyn_cup, bone2Name=pelvis, swing1LimitDegrees=45, swing2LimitDegrees=45, twistLimitDegrees=0}
 {
  bone1Name=thigh_l
  bone2Name=pelvis
  swing1LimitDegrees=45
  swing1Motion=Free
  swing2LimitDegrees=45
  swing2Motion=Free
  twistLimitDegrees=45
  twistMotion=Free
 }
]
