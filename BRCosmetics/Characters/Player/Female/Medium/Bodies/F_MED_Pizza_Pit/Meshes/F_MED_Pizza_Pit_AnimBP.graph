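# Animation Blueprint for the F_MED_Pizza_Pit cosmetic body, adding secondary motion to a pose copied from the attached parent mesh.
# Weighted constraints drive three left-shoulder shell bones, then AnimDynamics sways twelve shoelace bones at shoelace_alpha.
# The cup bone gets two offset-plus-AnimDynamics passes, weighted by crouch_alpha and crouch_alpha_invert, which crouch begin/end events set.
# A RigidBody node using the body's physics asset sits unconnected.
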AnimBlueprint /BRCosmetics/Characters/Player/Female/Medium/Bodies/F_MED_Pizza_Pit/Meshes/F_MED_Pizza_Pit_AnimBP
target_skeleton=&/F_MED_Pizza_Pit_Skeleton
graphs [
 AnimGraph {nodes [
   RigidBody() {
    bOverrideWorldGravity=true
    overridePhysicsAsset=&/F_MED_Pizza_Pit_Physics
    simulationSpace=BaseBoneSpace
    alphaBoolBlend {blendOption=Linear}
    baseBoneRef {boneName=root}
    overrideWorldGravity {z=-600}
   }
   n1 = Get crouch_alpha_invert() -> (crouch_alpha_invert:double)
   n2 = Get shoelace_alpha() -> (shoelace_alpha:double)
   n3 = Get shoelace_alpha()
   n4 = "Constraint - dyn_shell_03 (PPPPPP)"!(
    ComponentPose=Local To Component(LocalPose=Copy Pose From Mesh() {bUseAttachedParent=true})
    Alpha=1
    ConstraintWeights_0=0.2162
    ConstraintWeights_1:float=0.0005
    ConstraintWeights_2:float=0.495
    ConstraintWeights_3:float=0.2323
    ConstraintWeights_4:float=0.0555
    ConstraintWeights_5:float=0.0005
   ) -> (Pose) {
    alphaBoolBlend {blendOption=Linear}
    boneToModify {boneName=dyn_shell_03}
    constraintSetup [
     {targetBone {boneName=clavicle_l}}
     {targetBone {boneName=upperarm_twist_02_l}}
     {targetBone {boneName=spine_04}}
     {targetBone {boneName=spine_05}}
     {targetBone {boneName=upperarm_twist_01_l}}
     {targetBone {boneName=upperarm_l}}
    ]
    constraintWeights [1,1,1,1,1,1]
   }
   n5 = "Constraint - dyn_shell_02 (PPPPPP)"!(
    ComponentPose=@n4
    Alpha=1
    ConstraintWeights_0=0.4986
    ConstraintWeights_1:float=0.0002
    ConstraintWeights_2:float=0.2307
    ConstraintWeights_3:float=0.2156
    ConstraintWeights_4:float=0.0548
    ConstraintWeights_5:float=0.0002
   ) -> (Pose) {
    alphaBoolBlend {blendOption=Linear}
    boneToModify {boneName=dyn_shell_02}
    constraintSetup [
     {targetBone {boneName=clavicle_l}}
     {targetBone {boneName=upperarm_twist_02_l}}
     {targetBone {boneName=spine_04}}
     {targetBone {boneName=spine_05}}
     {targetBone {boneName=upperarm_twist_01_l}}
     {targetBone {boneName=upperarm_l}}
    ]
    constraintWeights [1,1,1,1,1,1]
   }
   n6 = "Constraint - dyn_shell_01 (PPPPPP)"!(
    ComponentPose=@n5
    Alpha=1
    ConstraintWeights_0=0.8048
    ConstraintWeights_1:float=0.0002
    ConstraintWeights_2:float=0.0855
    ConstraintWeights_3:float=0.0638
    ConstraintWeights_4:float=0.0456
    ConstraintWeights_5:float=0.0001
   ) -> (Pose) {
    alphaBoolBlend {blendOption=Linear}
    boneToModify {boneName=dyn_shell_01}
    constraintSetup [
     {targetBone {boneName=clavicle_l}}
     {targetBone {boneName=upperarm_twist_02_l}}
     {targetBone {boneName=spine_04}}
     {targetBone {boneName=spine_05}}
     {targetBone {boneName=upperarm_twist_01_l}}
     {targetBone {boneName=upperarm_l}}
    ]
    constraintWeights [1,1,1,1,1,1]
   }
   n7 = Anim Dynamics - Bone: dyn_string_outer_01_l(ComponentPose=@n6, Alpha=@n3, ExternalForce="0,0,0") {
    angularDampingOverride=1
    angularSpringConstant=10
    bAngularSpring=true
    bGravityOverrideInSimSpace=true
    bOverrideAngularDamping=true
    gravityScale=0.35
    simulationSpace=RootRelative
    alphaBoolBlend {blendOption=Linear}
    boundBone {boneName=dyn_string_outer_01_l}
    physicsBodyDefinitions [{
      sphereCollisionRadius=0
      boundBone {boneName=dyn_string_outer_01_l}
      boxExtents=<2 5 5>
      constraintSetup {angularLimitsMax=<0 20 0>, angularLimitsMin=<0 -20 -40>, angularTarget=<1 0 0>}
      localJointOffset=<2 0 0>
     }]
   }
   n8 = Anim Dynamics - Bone: dyn_string_outer_02_l(ComponentPose=@n7, Alpha=@n3, ExternalForce="0,0,0") {
    angularDampingOverride=1
    angularSpringConstant=10
    bAngularSpring=true
    bGravityOverrideInSimSpace=true
    bOverrideAngularDamping=true
    gravityScale=0.35
    simulationSpace=RootRelative
    alphaBoolBlend {blendOption=Linear}
    boundBone {boneName=dyn_string_outer_02_l}
    physicsBodyDefinitions [{
      sphereCollisionRadius=0
      boundBone {boneName=dyn_string_outer_02_l}
      boxExtents=<2 5 5>
      constraintSetup {angularLimitsMax=<0 20 0>, angularLimitsMin=<0 -20 -40>, angularTarget=<1 0 0>}
      localJointOffset=<2 0 0>
     }]
   }
   n9 = Anim Dynamics - Bone: dyn_string_outer_03_l(ComponentPose=@n8, Alpha=@n3, ExternalForce="0,0,0") {
    angularDampingOverride=1
    angularSpringConstant=10
    bAngularSpring=true
    bGravityOverrideInSimSpace=true
    bOverrideAngularDamping=true
    gravityScale=0.35
    simulationSpace=RootRelative
    alphaBoolBlend {blendOption=Linear}
    boundBone {boneName=dyn_string_outer_03_l}
    physicsBodyDefinitions [{
      sphereCollisionRadius=0
      boundBone {boneName=dyn_string_outer_03_l}
      boxExtents=<2 5 5>
      constraintSetup {angularLimitsMax=<0 20 0>, angularLimitsMin=<0 -20 -40>, angularTarget=<1 0 0>}
      localJointOffset=<2 0 0>
     }]
   }
   n10 = Anim Dynamics - Bone: dyn_string_inner_01_l(ComponentPose=@n9, Alpha=@n3, ExternalForce="0,0,0") {
    angularDampingOverride=1
    angularSpringConstant=10
    bAngularSpring=true
    bGravityOverrideInSimSpace=true
    bOverrideAngularDamping=true
    gravityScale=0.35
    simulationSpace=RootRelative
    alphaBoolBlend {blendOption=Linear}
    boundBone {boneName=dyn_string_inner_01_l}
    physicsBodyDefinitions [{
      sphereCollisionRadius=0
      boundBone {boneName=dyn_string_inner_01_l}
      boxExtents=<2 5 5>
      constraintSetup {angularLimitsMax=<0 20 0>, angularLimitsMin=<0 -20 -40>, angularTarget=<1 0 0>}
      localJointOffset=<2 0 0>
     }]
   }
   n11 = Anim Dynamics - Bone: dyn_string_inner_02_l(ComponentPose=@n10, Alpha=@n3, ExternalForce="0,0,0") {
    angularDampingOverride=1
    angularSpringConstant=10
    bAngularSpring=true
    bGravityOverrideInSimSpace=true
    bOverrideAngularDamping=true
    gravityScale=0.35
    simulationSpace=RootRelative
    alphaBoolBlend {blendOption=Linear}
    boundBone {boneName=dyn_string_inner_02_l}
    physicsBodyDefinitions [{
      sphereCollisionRadius=0
      boundBone {boneName=dyn_string_inner_02_l}
      boxExtents=<2 5 5>
      constraintSetup {angularLimitsMax=<0 20 0>, angularLimitsMin=<0 -20 -40>, angularTarget=<1 0 0>}
      localJointOffset=<2 0 0>
     }]
   }
   n12 = Anim Dynamics - Bone: dyn_string_inner_03_l(ComponentPose=@n11, Alpha=@n3, ExternalForce="0,0,0") {
    angularDampingOverride=1
    angularSpringConstant=10
    bAngularSpring=true
    bGravityOverrideInSimSpace=true
    bOverrideAngularDamping=true
    gravityScale=0.35
    simulationSpace=RootRelative
    alphaBoolBlend {blendOption=Linear}
    boundBone {boneName=dyn_string_inner_03_l}
    physicsBodyDefinitions [{
      sphereCollisionRadius=0
      boundBone {boneName=dyn_string_inner_03_l}
      boxExtents=<2 5 5>
      constraintSetup {angularLimitsMax=<0 20 0>, angularLimitsMin=<0 -20 -40>, angularTarget=<1 0 0>}
      localJointOffset=<2 0 0>
     }]
   }
   n13 = Anim Dynamics - Bone: dyn_string_inner_01_r(ComponentPose=@n12, Alpha=@n2, ExternalForce="0,0,0") {
    angularDampingOverride=1
    angularSpringConstant=10
    bAngularSpring=true
    bGravityOverrideInSimSpace=true
    bOverrideAngularDamping=true
    gravityScale=0.35
    simulationSpace=RootRelative
    alphaBoolBlend {blendOption=Linear}
    boundBone {boneName=dyn_string_inner_01_r}
    physicsBodyDefinitions [{
      sphereCollisionRadius=0
      boundBone {boneName=dyn_string_inner_01_r}
      boxExtents=<2 5 5>
      constraintSetup {angularLimitsMax=<0 20 0>, angularLimitsMin=<0 -20 -40>, angularTarget=<1 0 0>}
      localJointOffset=<2 0 0>
     }]
   }
   n14 = Anim Dynamics - Bone: dyn_string_inner_02_r(ComponentPose=@n13, Alpha=@n2, ExternalForce="0,0,0") {
    angularDampingOverride=1
    angularSpringConstant=10
    bAngularSpring=true
    bGravityOverrideInSimSpace=true
    bOverrideAngularDamping=true
    gravityScale=0.35
    simulationSpace=RootRelative
    alphaBoolBlend {blendOption=Linear}
    boundBone {boneName=dyn_string_inner_02_r}
    physicsBodyDefinitions [{
      sphereCollisionRadius=0
      boundBone {boneName=dyn_string_inner_02_r}
      boxExtents=<2 5 5>
      constraintSetup {angularLimitsMax=<0 20 0>, angularLimitsMin=<0 -20 -40>, angularTarget=<1 0 0>}
      localJointOffset=<2 0 0>
     }]
   }
   n15 = Anim Dynamics - Bone: dyn_string_inner_03_r(ComponentPose=@n14, Alpha=@n2, ExternalForce="0,0,0") {
    angularDampingOverride=1
    angularSpringConstant=10
    bAngularSpring=true
    bGravityOverrideInSimSpace=true
    bOverrideAngularDamping=true
    gravityScale=0.35
    simulationSpace=RootRelative
    alphaBoolBlend {blendOption=Linear}
    boundBone {boneName=dyn_string_inner_03_r}
    physicsBodyDefinitions [{
      sphereCollisionRadius=0
      boundBone {boneName=dyn_string_inner_03_r}
      boxExtents=<2 5 5>
      constraintSetup {angularLimitsMax=<0 20 0>, angularLimitsMin=<0 -20 -40>, angularTarget=<1 0 0>}
      localJointOffset=<2 0 0>
     }]
   }
   n16 = Anim Dynamics - Bone: dyn_string_outer_01_r(ComponentPose=@n15, Alpha=@n2, ExternalForce="0,0,0") {
    angularDampingOverride=1
    angularSpringConstant=10
    bAngularSpring=true
    bGravityOverrideInSimSpace=true
    bOverrideAngularDamping=true
    gravityScale=0.35
    simulationSpace=RootRelative
    alphaBoolBlend {blendOption=Linear}
    boundBone {boneName=dyn_string_outer_01_r}
    physicsBodyDefinitions [{
      sphereCollisionRadius=0
      boundBone {boneName=dyn_string_outer_01_r}
      boxExtents=<2 5 5>
      constraintSetup {angularLimitsMax=<0 20 0>, angularLimitsMin=<0 -20 -40>, angularTarget=<1 0 0>}
      localJointOffset=<2 0 0>
     }]
   }
   n17 = Anim Dynamics - Bone: dyn_string_outer_02_r(ComponentPose=@n16, Alpha=@n2, ExternalForce="0,0,0") {
    angularDampingOverride=1
    angularSpringConstant=10
    bAngularSpring=true
    bGravityOverrideInSimSpace=true
    bOverrideAngularDamping=true
    gravityScale=0.35
    simulationSpace=RootRelative
    alphaBoolBlend {blendOption=Linear}
    boundBone {boneName=dyn_string_outer_02_r}
    physicsBodyDefinitions [{
      sphereCollisionRadius=0
      boundBone {boneName=dyn_string_outer_02_r}
      boxExtents=<2 5 5>
      constraintSetup {angularLimitsMax=<0 20 0>, angularLimitsMin=<0 -20 -40>, angularTarget=<1 0 0>}
      localJointOffset=<2 0 0>
     }]
   }
   n18 = Anim Dynamics - Bone: dyn_string_outer_03_r(ComponentPose=@n17, Alpha=@n2, ExternalForce="0,0,0") {
    angularDampingOverride=1
    angularSpringConstant=10
    bAngularSpring=true
    bGravityOverrideInSimSpace=true
    bOverrideAngularDamping=true
    gravityScale=0.35
    simulationSpace=RootRelative
    alphaBoolBlend {blendOption=Linear}
    boundBone {boneName=dyn_string_outer_03_r}
    physicsBodyDefinitions [{
      sphereCollisionRadius=0
      boundBone {boneName=dyn_string_outer_03_r}
      boxExtents=<2 5 5>
      constraintSetup {angularLimitsMax=<0 20 0>, angularLimitsMin=<0 -20 -40>, angularTarget=<1 0 0>}
      localJointOffset=<2 0 0>
     }]
   }
   n19 = Get crouch_alpha() -> (crouch_alpha:double)
   n20 = "Transform (Modify) Bone - Bone: dyn_cup"(ComponentPose=@n18, Alpha=@n19, Translation="-0.224,-1.31,-1.165", Rotation="-1.344,-23.82,-60", Scale="1,1,1") {
    rotationMode=BMM_Additive
    rotationSpace=BCS_BoneSpace
    translationMode=BMM_Additive
    translationSpace=BCS_BoneSpace
    alphaBoolBlend {blendOption=Linear}
    boneToModify {boneName=dyn_cup}
    rotation=<rot -1.344 -23.82 -60>
    translation=<-0.224 -1.31 -1.165>
   }
   n21 = Anim Dynamics - Bone: dyn_cup(ComponentPose=@n20, Alpha=@n19, ExternalForce="0,0,0") {
    angularDampingOverride=1
    angularSpringConstant=1500
    bGravityOverrideInSimSpace=true
    bOverrideAngularDamping=true
    gravityScale=0.35
    simulationSpace=RootRelative
    alphaBoolBlend {blendOption=Linear}
    boundBone {boneName=dyn_cup}
    physicsBodyDefinitions [{
      sphereCollisionRadius=0
      boundBone {boneName=dyn_cup}
      boxExtents=<10 20 20>
      constraintSetup {angularLimitsMax=<0 20 0>, angularLimitsMin=<0 -20 -40>, angularTarget=<1 0 0>}
      localJointOffset=<10 0 0>
     }]
   }
   n22 = "Transform (Modify) Bone - Bone: dyn_cup"(ComponentPose=@n21, Alpha=@n1, Translation="-0.950,-1.231,-1.022", Rotation="-1.344,-23.82,-60", Scale="1,1,1") {
    rotationMode=BMM_Additive
    rotationSpace=BCS_BoneSpace
    translationMode=BMM_Additive
    translationSpace=BCS_BoneSpace
    alphaBoolBlend {blendOption=Linear}
    boneToModify {boneName=dyn_cup}
    rotation=<rot -1.344 -23.82 -60>
    translation=<-0.95 -1.231 -1.022>
   }
   n23 = Anim Dynamics - Bone: dyn_cup(ComponentPose=@n22, Alpha=@n1, ExternalForce="0,0,0") {
    angularDampingOverride=1
    angularSpringConstant=1500
    bAngularSpring=true
    bGravityOverrideInSimSpace=true
    bOverrideAngularDamping=true
    gravityScale=0.35
    simulationSpace=RootRelative
    alphaBoolBlend {blendOption=Linear}
    boundBone {boneName=dyn_cup}
    physicsBodyDefinitions [{
      sphereCollisionRadius=0
      boundBone {boneName=dyn_cup}
      boxExtents=<10 20 20>
      constraintSetup {angularLimitsMax=<0 20 40>, angularLimitsMin=<0 -20 -40>, angularTarget=<1 0 0>}
      localJointOffset=<10 0 0>
     }]
   }
   Output Pose(Result=Component To Local(ComponentPose=@n23)) {layerGroup=None}
  ]}
 EventGraph {nodes [
   Event OnBeginCrouching()
   > n1 = Set crouch_alpha(crouch_alpha:double=1) -> (Output_Get:double)
   > Set crouch_alpha_invert(crouch_alpha_invert:double=float - float(B=@n1) {is=title K2Node_CallFunction}) -> (Output_Get:double)
   Event OnEndCrouching()
   > n2 = Set crouch_alpha(crouch_alpha=0)
   > Set crouch_alpha_invert(crouch_alpha_invert=float - float(B=@n2))
   Event BlueprintUpdateAnimation()
  ]}
]
variables [
 crouch_alpha {type=number}
 crouch_alpha_invert {type=number, default=1}
 shoelace_alpha {type=number, default=0.6}
]
